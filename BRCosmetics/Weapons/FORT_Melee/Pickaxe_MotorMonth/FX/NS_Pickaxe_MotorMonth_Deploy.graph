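# Niagara deploy effect for the MotorMonth pickaxe: green lightning, edge glow with lights, and a SpiderMask sprite pair.
# Three lightning emitters burst hand-placed meshes via partition-indexed arrays, animating dynamic material parameters with random values and curves, plus a DarkColor attribute.
# Delayed EdgeGlow and SpiderMask emitters follow the lightning, each spawning two particles with mirrored offsets.
# Curve-driven intensities are scaled by time of day and front end, plus platform on Lightninghalf.
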NiagaraSystem /BRCosmetics/Weapons/FORT_Melee/Pickaxe_MotorMonth/FX/NS_Pickaxe_MotorMonth_Deploy
dependencies {
 usedDataInterfaces [&NiagaraDataInterfaceArrayFloat3, &NiagaraDataInterfaceCurve, &NiagaraDataInterfacePlatformSet]
 usedDynamicInputs []
 usedModules [&SystemState]
}
emitters {
 EdgeGlow {
  emitterUpdateScript [
   &EmitterState(
    Inactive Response=::NewEnumerator1
    Life Cycle Mode=::NewEnumerator1
    Loop Behavior=::NewEnumerator1
    Loop Delay=0.7
    Loop Duration=0.23
    UseLoopDelay=true
   )
   &SpawnBurst_Instantaneous(Spawn Count=2, Use Spawn Probability=false, UseLoopCountLimit=false)
  ]
  particleSpawnScript [
   &PartitionParticles()
   &InitializeParticle(
    Color=<rgba 1 1 1 0.8>
    Lifetime=5
    Mesh Scale=<1 1.5 1>
    Mesh Scale Mode=::NewEnumerator4
    Position Offset=&SelectVectorFromArray(
     Direct Array Index=$Particles.PartitionParticles.Partition
     Vector Selection Array=DataInterface() {propertyValues {FloatData [<0 0.05 -0.05>, <0 0.05 0.05>]}}
    )
    UsePositionOffset=true
    Write Material Random=true
   )
   &InitialMeshOrientation(
    Mesh Orientation Mode=::NewEnumerator2
    Rotation=&SelectVectorFromArray(
     Direct Array Index=$Particles.PartitionParticles.Partition
     Vector Selection Array=DataInterface() {propertyValues {FloatData [<0 0 0>, <0 0 0.5>]}}
    )
    Use Rotation Vector=true
   )
   &InitialMeshOrientation(
    Mesh Orientation Mode=::NewEnumerator2
    Rotation=&SelectVectorFromArray(
     Direct Array Index=$Particles.PartitionParticles.Partition
     Vector Selection Array=DataInterface() {propertyValues {FloatData [<0 0.25 0>, <0 0.25 0>]}}
    )
    Use Rotation Vector=true
   ) {module=InitialMeshOrientation001}
  ]
  particleUpdateScript [
   &ParticleState()
   &FN_Light_Attributes(
    Color=<rgba 0.0469 1 0.112 1>
    Create Color Attribute=true
    Create Exponent Attribute=true
    Create Position Attribute=true
    Create Radius Attribute=true
    Exponent=3
    Position=&SelectVectorFromArray(
     Direct Array Index=$Particles.PartitionParticles.Partition
     Vector Selection Array=DataInterface() {propertyValues {FloatData [<-10 0 20>, <-10 0 -20>]}}
    )
    Radius=90
    Scale Color=&FloatFromCurve(FloatCurve=DataInterface() {propertyValues {Curve {keys [<at 0 0>,<at 0.05 1>]}}}, Scale Curve=10)
   )
   &ScaleColor(
    Scale Alpha=&Multiply_Float(
     A=&FNTimeOfDayScalar(Night Scale=0.5, Morning Scale=1, Evening Scale=0.8)
     B=&Multiply_Float(
      A=&FN_FrontEndFloatModifier(Gameplay=1, Front End=0.6)
      B=&FloatFromCurve(FloatCurve=DataInterface() {propertyValues {Curve {keys [<at 0 0>,<at 0.05 1>]}}})
     )
    )
    Scale RGB=&MakeVectorFromLinearColorRGB(Color=<rgba 0.2173 3 0.3388 1>)
    ScaleA=true
    ScaleRGB=true
   )
  ]
  rendererClasses [&NiagaraMeshRendererProperties, &NiagaraLightRendererProperties]
  renderers [
   {rendererClass=&NiagaraMeshRendererProperties, rendererIndex=0}
   {rendererClass=&NiagaraLightRendererProperties, rendererIndex=1}
  ]
 }
 LightningSix {
  emitterUpdateScript [
   &EmitterState(
    Enable Distance Culling=true
    Inactive Response=::NewEnumerator1
    Life Cycle Mode=::NewEnumerator1
    Loop Behavior=::NewEnumerator1
    Loop Delay=0.2
    Loop Duration=0.7
    Module.bUseMaxDistance=true
    Module.bUseMinDistance=true
    Scalability Mode=::NewEnumerator1
    UseLoopDelay=true
   )
   &SpawnBurst_Instantaneous(Spawn Count=10, Use Spawn Probability=false, UseLoopCountLimit=false)
  ]
  particleSpawnScript [
   &PartitionParticles(Number Of Partitions=10)
   &InitializeParticle(
    Color=<rgba 0.2891 1.5 0.2955 1>
    Lifetime=5
    Mesh Scale=&SelectVectorFromArray(
     Direct Array Index=$Particles.PartitionParticles.Partition
     Vector Selection Array=DataInterface() {propertyValues {FloatData [<0 0 0>, <1.4 1 1>, <1.2 1 1>, <0.8 1 1>, <0 0 0>, <0 0 0>, <1.2 1 -1>, <1 1 -1>, <0.85 1 -1>, <0 0 0>]}}
    )
    Mesh Scale Mode=::NewEnumerator4
    Position Offset=&SelectVectorFromArray(
     Direct Array Index=$Particles.PartitionParticles.Partition
     Vector Selection Array=DataInterface() {propertyValues {FloatData [<1 0 -15>, <0 0 -10>, <2 0 -10>, <0 0 -13>, <0 0 -15>, <0 0 15>, <0 0 15>, <0 0 13>, <0 0 12>, <0 0 13>]}}
    )
    UsePositionOffset=true
    Write Material Random=true
   )
   &InitialMeshOrientation(
    Mesh Orientation Mode=::NewEnumerator2
    Rotation=&SelectVectorFromArray(
     Direct Array Index=$Particles.PartitionParticles.Partition
     Vector Selection Array=DataInterface() {propertyValues {FloatData [<0 -0.017 0>, <0 0.005 0>, <0 0.09 0>, <0 0.21 0>, <0 0.345 0>, <0 0.015 0>, <0 -0.03 0>, <0 -0.1 0>, <0 -0.21 0>, <0 -0.35 0>]}}
    )
    Use Rotation Vector=true
   )
   &InitialMeshOrientation(
    Mesh Orientation Mode=::NewEnumerator2
    Rotation=&SelectVectorFromArray(
     Direct Array Index=$Particles.PartitionParticles.Partition
     Vector Selection Array=DataInterface() {propertyValues {FloatData [<0 0.25 0>, <0 0.25 0>, <0 0.25 0>, <0 0.25 0>, <0 0.25 0>, <0 -0.25 0>, <0 -0.25 0>, <0 -0.25 0>, <0 -0.25 0>, <0 -0.25 0>]}}
    )
    Use Rotation Vector=true
   ) {module=InitialMeshOrientation001}
  ]
  particleUpdateScript [
   &ParticleState()
   &DynamicMaterialParameters(
    Index 2 Param 1=&RandomRangeFloat(Minimum=0.7, Maximum=0.8)
    Index 2 Param 3=&RandomRangeFloat(Minimum=0.2, Maximum=0.8)
    Index 2 Param 4=&RandomRangeFloat(Minimum=0.2, Maximum=0.8)
    Index 3 Param 1=&RandomRangeFloat(Minimum=0.8, Maximum=1)
    Index 3 Param 2=&RandomRangeFloat(Minimum=0.8, Maximum=1)
    Index 3 Param 4=&FloatFromCurve(
     FloatCurve=DataInterface() {propertyValues {Curve {keys [
         {arriveTangent=2.4891, interpMode=RCIM_Cubic, leaveTangent=2.4891, tangentMode=RCTM_User, time=0, value=0}
         {arriveTangent=2.8022, interpMode=RCIM_Cubic, leaveTangent=2.8022, tangentMode=RCTM_User, time=0.15, value=1}
        ]}}}
     Scale Curve=50
    )
    Module.Index2Param0Write=true
    Module.Index2Param1Write=true
    Module.Index2Param2Write=true
    Module.Index2Param3Write=true
    Module.Index3Param0Write=true
    Module.Index3Param1Write=true
    Module.Index3Param2Write=false
    Module.Index3Param3Write=true
    Write Parameter Index 0=false
    Write Parameter Index 2=true
    Write Parameter Index 3=true
   )
   &ScaleColor(
    Scale Alpha=&Multiply_Float(
     A=&FNTimeOfDayScalar(Night Scale=0.7, Morning Scale=1, Evening Scale=1)
     B=&Multiply_Float(
      A=&FN_FrontEndFloatModifier(Gameplay=1, Front End=0.7)
      B=&FloatFromCurve(FloatCurve=DataInterface() {propertyValues {Curve {keys [
           <at 0 0>
           <at 0.1 1>
           <at 0.12 1>
           {arriveTangent=-114.4434, interpMode=RCIM_Cubic, leaveTangent=-114.4434, tangentMode=RCTM_User, time=0.13, value=5}
           {interpMode=RCIM_Cubic, leaveTangent=-60, time=0.15, value=1}
          ]}}})
     )
    )
    ScaleA=true
    ScaleRGB=true
   )
   /SetVariables(Particles.DarkColor:LinearColor=&MultiplyLinearColor(
     Linear Color=<rgba 0.0237 0.1667 0.0208 0.7>
     Scale Factor=&MakeVector4_FromFloat(Float=&Multiply_Float(
       A=&FNTimeOfDayScalar(Night Scale=0.7, Morning Scale=1, Evening Scale=1)
       B=&Multiply_Float(
        A=&FN_FrontEndFloatModifier(Gameplay=1, Front End=0.7)
        B=&FloatFromCurve(FloatCurve=DataInterface() {propertyValues {Curve {keys [<at 0 0>,<at 0.05 1>,<at 0.12 1>,<at 0.14 2>,<at 0.17 1>]}}})
       )
      ))
    )) {isSetParameters=true, module=SetVariables_#1}
  ]
  rendererClasses [&NiagaraMeshRendererProperties]
  renderers [
   {rendererClass=&NiagaraMeshRendererProperties, rendererIndex=0}
   {rendererClass=&NiagaraMeshRendererProperties, rendererIndex=1}
  ]
 }
 LightningWide {
  emitterUpdateScript [
   &EmitterState(
    Enable Distance Culling=true
    Inactive Response=::NewEnumerator1
    Life Cycle Mode=::NewEnumerator1
    Loop Behavior=::NewEnumerator1
    Loop Delay=0.2
    Loop Duration=0.7
    Module.bUseMaxDistance=true
    Module.bUseMinDistance=true
    Scalability Mode=::NewEnumerator1
    UseLoopDelay=true
   )
   &SpawnBurst_Instantaneous(Spawn Count=10, Use Spawn Probability=false, UseLoopCountLimit=false)
  ]
  particleSpawnScript [
   &PartitionParticles(Number Of Partitions=10)
   &InitializeParticle(
    Color=<rgba 0.1927 1 0.197 1>
    Lifetime=5
    Mesh Scale=&SelectVectorFromArray(
     Direct Array Index=$Particles.PartitionParticles.Partition
     Vector Selection Array=DataInterface() {propertyValues {FloatData [<1.3 1 2>, <0 0 0>, <0 0 0>, <0 0 0>, <1 1 1>, <1.3 1 -2>, <0 0 0>, <0 0 0>, <0 0 0>, <1 1 -1.5>]}}
    )
    Mesh Scale Mode=::NewEnumerator4
    Position Offset=&SelectVectorFromArray(
     Direct Array Index=$Particles.PartitionParticles.Partition
     Vector Selection Array=DataInterface() {propertyValues {FloatData [<1 0 -15>, <0 0 -10>, <0 0 -10>, <0 0 -15>, <1 0 -15>, <0 0 15>, <0 0 15>, <-1 0 15>, <0 0 13>, <0 0 13>]}}
    )
    UsePositionOffset=true
    Write Material Random=true
   )
   &InitialMeshOrientation(
    Mesh Orientation Mode=::NewEnumerator2
    Rotation=&SelectVectorFromArray(
     Direct Array Index=$Particles.PartitionParticles.Partition
     Vector Selection Array=DataInterface() {propertyValues {FloatData [<0 -0.01 0>, <0 0.005 0>, <0 0.09 0>, <0 0.24 0>, <0 0.345 0>, <0 0.01 0>, <0 -0.03 0>, <0 -0.1 0>, <0 -0.21 0>, <0 -0.35 0>]}}
    )
    Use Rotation Vector=true
   )
   &InitialMeshOrientation(
    Mesh Orientation Mode=::NewEnumerator2
    Rotation=&SelectVectorFromArray(
     Direct Array Index=$Particles.PartitionParticles.Partition
     Vector Selection Array=DataInterface() {propertyValues {FloatData [<0 0.25 0>, <0 0.25 0>, <0 0.25 0>, <0 0.25 0>, <0 0.25 0>, <0 -0.25 0>, <0 -0.25 0>, <0 -0.25 0>, <0 -0.25 0>, <0 -0.25 0>]}}
    )
    Use Rotation Vector=true
   ) {module=InitialMeshOrientation001}
  ]
  particleUpdateScript [
   &ParticleState()
   &DynamicMaterialParameters(
    Index 2 Param 1=&RandomRangeFloat(Minimum=1, Maximum=1)
    Index 2 Param 3=&RandomRangeFloat(Minimum=0.2, Maximum=0.8)
    Index 2 Param 4=&RandomRangeFloat(Minimum=0.2, Maximum=0.8)
    Index 3 Param 1=&RandomRangeFloat(Minimum=0.8, Maximum=1)
    Index 3 Param 2=&RandomRangeFloat(Minimum=0.8, Maximum=1)
    Index 3 Param 4=&FloatFromCurve(
     FloatCurve=DataInterface() {propertyValues {Curve {keys [
         {arriveTangent=10.2598, interpMode=RCIM_Cubic, leaveTangent=10.2598, tangentMode=RCTM_User, time=0, value=0}
         {arriveTangent=16.7192, interpMode=RCIM_Cubic, leaveTangent=16.7192, tangentMode=RCTM_User, time=0.15, value=1}
        ]}}}
     Scale Curve=50
    )
    Module.Index2Param0Write=true
    Module.Index2Param1Write=true
    Module.Index2Param2Write=true
    Module.Index2Param3Write=true
    Module.Index3Param0Write=true
    Module.Index3Param1Write=true
    Module.Index3Param2Write=false
    Module.Index3Param3Write=true
    Write Parameter Index 0=false
    Write Parameter Index 2=true
    Write Parameter Index 3=true
   )
   &ScaleColor(
    Scale Alpha=&Multiply_Float(
     A=&Multiply_Float(
      A=&FN_FrontEndFloatModifier(Gameplay=1, Front End=0.8)
      B=&FNTimeOfDayScalar(Night Scale=0.7, Morning Scale=1, Evening Scale=1)
     )
     B=&FloatFromCurve(FloatCurve=DataInterface() {propertyValues {Curve {keys [
          <at 0 0>
          <at 0.05 1>
          <at 0.12 1>
          {arriveTangent=-133.6192, interpMode=RCIM_Cubic, leaveTangent=-133.6192, tangentMode=RCTM_User, time=0.13, value=3}
          <at 0.15 1 RCIM_Cubic>
         ]}}})
    )
    ScaleA=true
    ScaleRGB=true
   )
   /SetVariables(Particles.DarkColor:LinearColor=&MultiplyLinearColor(
     Linear Color=<rgba 0.0237 0.1667 0.0208 0.7>
     Scale Factor=&MakeVector4_FromFloat(Float=&Multiply_Float(
       A=&FNTimeOfDayScalar(Night Scale=0.5, Morning Scale=1, Evening Scale=1)
       B=&Multiply_Float(
        A=&FN_FrontEndFloatModifier(Gameplay=1, Front End=0.7)
        B=&FloatFromCurve(FloatCurve=DataInterface() {propertyValues {Curve {keys [<at 0 0>,<at 0.2 1>]}}})
       )
      ))
    )) {isSetParameters=true, module=SetVariables_#1}
  ]
  rendererClasses [&NiagaraMeshRendererProperties]
  renderers [
   {rendererClass=&NiagaraMeshRendererProperties, rendererIndex=0}
   {rendererClass=&NiagaraMeshRendererProperties, rendererIndex=1}
  ]
 }
 Lightninghalf {
  emitterUpdateScript [
   &EmitterState(
    Enable Distance Culling=true
    Inactive Response=::NewEnumerator1
    Life Cycle Mode=::NewEnumerator1
    Loop Behavior=::NewEnumerator1
    Loop Delay=0.15
    Loop Duration=0.7
    Module.bUseMaxDistance=true
    Module.bUseMinDistance=true
    Scalability Mode=::NewEnumerator1
    UseLoopDelay=true
   )
   &SpawnBurst_Instantaneous(Spawn Count=6, Use Spawn Probability=false, UseLoopCountLimit=false)
  ]
  particleSpawnScript [
   &PartitionParticles(Number Of Partitions=6)
   &InitializeParticle(
    Color=<rgba 0.2698 1.4 0.2758 1>
    Lifetime=5
    Mesh Scale=&SelectVectorFromArray(
     Direct Array Index=$Particles.PartitionParticles.Partition
     Vector Selection Array=DataInterface() {propertyValues {FloatData [<1.05 1.1 1.12>, <0.7 1 0.75>, <0.4 0.4 0.4>, <1.1 1 1.15>, <0.7 0.7 0.8>, <0.4 0.45 0.45>]}}
    )
    Mesh Scale Mode=::NewEnumerator4
    Position Offset=&SelectVectorFromArray(
     Direct Array Index=$Particles.PartitionParticles.Partition
     Vector Selection Array=DataInterface() {propertyValues {FloatData [<-13 0 -7>, <-8 0 -9>, <-4 0 -10>, <-14 0 6.5>, <-9 0 10>, <-5 0 11>]}}
    )
    UsePositionOffset=true
    Write Material Random=true
   )
   &InitialMeshOrientation(
    Mesh Orientation Mode=::NewEnumerator2
    Rotation=&SelectVectorFromArray(
     Direct Array Index=$Particles.PartitionParticles.Partition
     Vector Selection Array=DataInterface() {propertyValues {FloatData [<0 0 0>, <0 0 0>, <0 0 0>, <0.5 0 0>, <0.5 0 0>, <0.5 0 0>]}}
    )
    Use Rotation Vector=true
   )
   &InitialMeshOrientation(
    Mesh Orientation Mode=::NewEnumerator2
    Rotation=&SelectVectorFromArray(
     Direct Array Index=$Particles.PartitionParticles.Partition
     Vector Selection Array=DataInterface() {propertyValues {FloatData [<0 0.25 0>, <0 0.25 0>, <0 0.25 0>, <0 -0.25 0>, <0 -0.25 0>, <0 -0.25 0>]}}
    )
    Use Rotation Vector=true
   ) {module=InitialMeshOrientation001}
  ]
  particleUpdateScript [
   &ParticleState()
   &DynamicMaterialParameters(
    Index 2 Param 1=&RandomRangeFloat(Minimum=0.5, Maximum=0.5)
    Index 2 Param 3=&RandomRangeFloat(Minimum=0.2, Maximum=0.8)
    Index 2 Param 4=&RandomRangeFloat(Minimum=0.2, Maximum=0.8)
    Index 3 Param 1=&RandomRangeFloat(Minimum=0.8, Maximum=1)
    Index 3 Param 2=&RandomRangeFloat(Minimum=0.8, Maximum=0.5)
    Index 3 Param 4=&FloatFromCurve(
     FloatCurve=DataInterface() {propertyValues {Curve {keys [
         {arriveTangent=9.2278, interpMode=RCIM_Cubic, leaveTangent=9.2278, tangentMode=RCTM_User, time=0.04, value=0}
         {arriveTangent=6.8421, interpMode=RCIM_Cubic, leaveTangent=6.8421, tangentMode=RCTM_User, time=0.15, value=1}
        ]}}}
     Scale Curve=60
    )
    Module.Index2Param0Write=true
    Module.Index2Param1Write=true
    Module.Index2Param2Write=true
    Module.Index2Param3Write=true
    Module.Index3Param0Write=true
    Module.Index3Param1Write=true
    Module.Index3Param2Write=false
    Module.Index3Param3Write=true
    Write Parameter Index 0=false
    Write Parameter Index 2=true
    Write Parameter Index 3=true
   )
   &ScaleColor(
    Scale Alpha=&Multiply_Float(
     A=&Multiply_Float(
      A=&FNTimeOfDayScalar(Night Scale=0.7, Morning Scale=1, Evening Scale=1)
      B=&Multiply_Float(
       A=&FN_FrontEndFloatModifier(Gameplay=1, Front End=0.8)
       B=&FN_PickFloatDependOnPlatform(TargetPlatformActive=1, TargetPlatformNoActive=0.6)
      )
     )
     B=&FloatFromCurve(FloatCurve=DataInterface() {propertyValues {Curve {keys [
          <at 0 0>
          <at 0.12 1>
          {arriveTangent=-109.9987, interpMode=RCIM_Cubic, leaveTangent=-109.9987, tangentMode=RCTM_User, time=0.13, value=3.5}
          <at 0.15 1 RCIM_Cubic>
         ]}}})
    )
    ScaleA=true
    ScaleRGB=true
   )
   /SetVariables(Particles.DarkColor:LinearColor=&MultiplyLinearColor(
     Linear Color=<rgba 0.0237 0.1667 0.0208 0.7>
     Scale Factor=&MakeVector4_FromFloat(Float=&Multiply_Float(
       A=&Multiply_Float(
        A=&FNTimeOfDayScalar(Night Scale=0.7, Morning Scale=1, Evening Scale=1)
        B=&Multiply_Float(
         A=&FN_FrontEndFloatModifier(Gameplay=1, Front End=0.6)
         B=&FN_PickFloatDependOnPlatform(TargetPlatformActive=1, TargetPlatformNoActive=0.75)
        )
       )
       B=&FloatFromCurve(FloatCurve=DataInterface() {propertyValues {Curve {keys [<at 0 0>,<at 0.2 1>]}}})
      ))
    )) {isSetParameters=true}
  ]
  rendererClasses [&NiagaraMeshRendererProperties]
  renderers [
   {rendererClass=&NiagaraMeshRendererProperties, rendererIndex=0}
   {rendererClass=&NiagaraMeshRendererProperties, rendererIndex=1}
  ]
 }
 SpiderMask {
  emitterUpdateScript [
   &EmitterState(
    Inactive Response=::NewEnumerator1
    Life Cycle Mode=::NewEnumerator1
    Loop Behavior=::NewEnumerator1
    Loop Delay=0.7
    Loop Duration=0.23
    UseLoopDelay=true
   )
   &SpawnBurst_Instantaneous(Spawn Count=2, Use Spawn Probability=false, UseLoopCountLimit=false)
  ]
  particleSpawnScript [
   &PartitionParticles()
   &InitializeParticle(
    Position Offset=&SelectVectorFromArray(
     Direct Array Index=$Particles.PartitionParticles.Partition
     Vector Selection Array=DataInterface() {propertyValues {FloatData [<-5 0 -23>, <-5 0 23>]}}
    )
    Sprite Rotation Angle=10
    Sprite Rotation Mode=::NewEnumerator0
    Sprite Size=<35 35>
    Sprite Size Mode=::NewEnumerator4
    UsePositionOffset=true
    Write Material Random=true
   )
   &SpriteFacingAndAlignment(
    Sprite Alignment=<1 0 0>
    Sprite Facing=&SelectVectorFromArray(
     Direct Array Index=$Particles.PartitionParticles.Partition
     Vector Selection Array=DataInterface() {propertyValues {FloatData [<0 1 0>, <0 -1 0>]}}
    )
    WriteAlignment=true
    WriteFacing=true
   )
  ]
  particleUpdateScript [
   &ParticleState()
   &ScaleColor(
    Scale Alpha=&Multiply_Float(
     A=&Multiply_Float(A=0.2, B=&FNTimeOfDayScalar(Night Scale=0.5, Morning Scale=1, Evening Scale=1))
     B=&Multiply_Float(A=&FN_FrontEndFloatModifier(Gameplay=1, Front End=0.3), B=1)
    )
    Scale RGB=&MakeVectorFromLinearColorRGB(Color=<rgba 0 5 0.2559 1>)
    ScaleA=true
    ScaleRGB=true
   )
   &ScaleColor(
    Scale Alpha=&FloatFromCurve(FloatCurve=DataInterface() {propertyValues {Curve {keys [
         {arriveTangent=-6.6667, interpMode=RCIM_Cubic, time=0, value=0}
         {arriveTangent=0.7493, interpMode=RCIM_Cubic, leaveTangent=0.7493, tangentMode=RCTM_User, time=0.0993, value=1.947}
         {arriveTangent=-3.5038, interpMode=RCIM_Cubic, leaveTangent=-3.5038, tangentMode=RCTM_User, time=0.3, value=1}
        ]}}})
    ScaleA=true
    ScaleRGB=true
   ) {module=ScaleColor001}
  ]
  rendererClasses [&NiagaraSpriteRendererProperties]
  renderers [{rendererClass=&NiagaraSpriteRendererProperties}]
 }
}
summary {
 systemName=NS_Pickaxe_MotorMonth_Deploy
 emitters [
  {emitterName=EdgeGlow, rendererClasses [&NiagaraMeshRendererProperties, &NiagaraLightRendererProperties]}
  {emitterName=LightningSix, rendererClasses [&NiagaraMeshRendererProperties]}
  {emitterName=Lightninghalf, rendererClasses [&NiagaraMeshRendererProperties]}
  {emitterName=SpiderMask, rendererClasses [&NiagaraSpriteRendererProperties]}
  {emitterName=LightningWide, rendererClasses [&NiagaraMeshRendererProperties]}
 ]
 userVariables [
  User.bIsFrontEnd {type=NiagaraBool}
  User.bIsFrontEndPreview {type=NiagaraBool}
 ]
}
